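; Ashton Indigo repulsor projectile: a homing shot that explodes on impact or when its authority-side lifetime timer runs out.
; On remote clients, BeginPlay feeds the firing hand's socket into tracer material parameters; stopped projectiles linger briefly before destruction.
; Setting Impact Location drives an OnRep that hides the tracer and spawns explosion VFX, bone-attached with hit decals on players.
; Homing switches off after passing or nearing the target.
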(event Custom|Play%20Explosion%20Sound
  ;[1 1 1] Event Play Explosion Sound
  (Utilities|FlowControl|DoOnce false
    (:Completed
      (Audio|SpawnSoundatLocation "/Game/Athena/Sounds/LTMs/Ashton/Indigo/Ashton_Indigo_Proj_Expl_Cue.Ashton_Indigo_Proj_Expl_Cue" (Transformation|GetActorLocation)))))

(event Custom|Explosion%20VFX
  (bind (_bblockinghit _binitialoverlap _time _distance _location _impactpoint _normal _impactnormal _physmat _hitactor) (Collision|BreakHitResult (Variables|Default|GetHitResult)))
  (bind _hitresult (Variables|Default|GetHitResult))
  ;[1 1 1] Event Spawn Explosion VFX
  (bind _asfort_player_pawn_athena (Utilities|Casting|CastToFortPlayerPawnAthena _hitactor)
    (:then
      (bind _returnvalue (Effects|SpawnEmitterAttached "/Game/Athena/Items/Gameplay/BackPacks/Ashton/Indigo/FX/P_Ashton_Indigo_Explosion_PlayerImpact.P_Ashton_Indigo_Explosion_PlayerImpact" (Variables|/Script/Engine.Character|GetMesh _asfort_player_pawn_athena) (select (Utilities|Name|Equal(Name) (.ReturnValue (Components|SkinnedMesh|FindClosestBone (Variables|/Script/Engine.Character|GetMesh _asfort_player_pawn_athena) (Variables|VFX|GetImpact%20Location) 0.0 false)) "head") "neck_01" (.ReturnValue (Components|SkinnedMesh|FindClosestBone (Variables|/Script/Engine.Character|GetMesh _asfort_player_pawn_athena) (Variables|VFX|GetImpact%20Location) 0.0 false)))))
      (Transformation|SetWorldRotation _returnvalue (Math|Conversions|RotationFromXVector (* (Transformation|GetActorForwardVector) -1.0)) false false)
      (bind _returnvalue_2 (Rendering|Material|CreateNamedDynamicMaterialInstance _returnvalue "Decals" "/Game/Athena/Items/Gameplay/BackPacks/Ashton/Indigo/FX/M_Ashton_Indigo_PlayerHit_Decals.M_Ashton_Indigo_PlayerHit_Decals"))
      (Rendering|Material|SetVectorParameterValue _returnvalue_2 "HitLocation" (Math|Conversions|ToLinearColor(Vector) (Variables|VFX|GetImpact%20Location)) :!class /Script/Engine.MaterialInstanceDynamic))
    (:CastFailed
      (Effects|SpawnEmitteratLocation "/Game/Athena/Items/Gameplay/BackPacks/Ashton/Indigo/FX/P_Ashton_Indigo_Explosion.P_Ashton_Indigo_Explosion" (Variables|VFX|GetImpact%20Location) (Math|Rotator|MakeRotfromX _impactnormal)))))

(event EventOnStop
  (params (Hit :type Hit%20Result%20Structure%20%28by%20ref%29))
  (bind (_bblockinghit _binitialoverlap _time _distance _location _impactpoint) (Collision|BreakHitResult Hit))
  ;[1 1 1] Event on Stop
  (Variables|Default|SetHitResult Hit)
  (Variables|VFX|SetImpact%20Location _impactpoint)
  (Utilities|FlowControl|Sequence
    (:then_0
      (Utilities|FlowControl|DoOnce false
        (:Completed
          (Variables|VFX|SetbHasExploded? true)
          (CallFunction|PlayExplosionSound)
          (Delivery|DoExplosion) ; Explosion
        )))
    (:then_1
      (Networking|SwitchHasAuthority
        (:Authority
          (Utilities|Time|ClearandInvalidateTimerbyHandle (Variables|Default|GetLifespan%20Timer)))
        (:Remote
          (Utilities|FlowControl|Delay (Variables|VFX|GetFX%20Linger%20Time))
          (Actor|DestroyActor))))))

(event Custom|Reset%20Explosion%20Sound
  ;[1 1 1] Event Play Explosion Sound
  (Utilities|FlowControl|DoOnce false
    (:Completed
      (Audio|SpawnSoundatLocation "/Game/Athena/Sounds/LTMs/Ashton/Indigo/Ashton_Indigo_Proj_Expl_Cue.Ashton_Indigo_Proj_Expl_Cue" (Transformation|GetActorLocation)))))

(event Custom|Timeout
  (Components|Movement|StopMovementImmediately (Variables|Movement|GetProjectileMovementComponent))
  (if (not (Variables|VFX|GetbHasExploded?))
    (Variables|VFX|SetImpact%20Location (Transformation|GetWorldLocation (Variables|VFX|GetTracerBody)))
    (Variables|VFX|SetbHasExploded? true)
    (Delivery|DoExplosion) ; Explosion
  ))

(event EventBeginPlay
  (bind _tracermesh (Variables|Default|GetTracerMesh))
  (Utilities|FlowControl|Sequence
    (:then_0
      (Variables|VFX|SetFX%20Linger%20Time (Math|Random|RandomFloatinRange 0.5 1.0))
      (Networking|SwitchHasAuthority
        (:Authority
          (bind _returnvalue (Utilities|Time|SetTimerbyEvent (delegate Custom|Timeout) (Variables|Default|GetProjectile%20Max%20Lifetime)))
          (Variables|Default|SetLifespan%20Timer _returnvalue))
        (:Remote
          (bind _asb_ranged_ashton_indigo_athena (Utilities|Casting|CastToB_Ranged_Ashton_Indigo_Athena (Actor|GetOwner)))
          (bind _asfort_player_pawn (Utilities|Casting|CastToFortPlayerPawn (Variables|Actor|GetInstigator)))
          (bind (_pawn _part _meshpart) (Cosmetics|WaitForCosmeticSpecificPartApplied _asfort_player_pawn true "MiscOrTail")
            (:OnPartApplied
              (Rendering|Material|SetVectorParameterValueonMaterials _tracermesh "BulletOrigin" (Transformation|GetSocketLocation _meshpart (Utilities|Select :Index (Variables|/Game/Athena/Items/Gameplay/BackPacks/Ashton/Indigo/B_Ranged_Ashton_Indigo_Athena.B_Ranged_Ashton_Indigo_Athena_C|GetLastHandFired _asb_ranged_ashton_indigo_athena) :LEFT "FX_repulsor_l" :RIGHT "FX_repulsor_r")))
              (Rendering|Material|SetScalarParameterValueonMaterials _tracermesh "StartTime" (Utilities|Time|GetGameTimeinSeconds))
              (Rendering|Material|SetScalarParameterValueonMaterials _tracermesh "WPODuration" (Variables|Default|GetTracer%20Lifetime)))))))
    (:then_1
      (CallFunction|ResetExplosionSound))))

(event ProjectileMovement|EventOnHomingPassedTarget
  (Projectile|SetHomingStyle 0 0 "None" false :!class /Script/FortniteGame.FortProjectileMovementInterface))

(event ProjectileMovement|EventOnReachedLockTargetDistanceThreshold
  (Projectile|SetHomingStyle 0 0 "None" false :!class /Script/FortniteGame.FortProjectileMovementInterface))

(fn OnRep_ImpactLocation
  (Components|Activation|Deactivate (Variables|VFX|GetTracerBody))
  (Networking|CanExecuteCosmeticEvents
    (:True
      (CallFunction|ExplosionVFX))))

(fn ConstructionScript)
