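; Dual-hand Ashton Indigo ranged weapon Blueprint that fires lock-on homing projectiles.
; On local equip it creates a lock-on widget, passes it the lock-on target, and animates its material through timelines on new target and lock end.
; With a valid target, the server configures each spawned projectile's homing toward it.
; Fire events track the last hand fired and drive force feedback, using a gate marked as a temporary fix for blocked shots.
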(event Custom|BindOnHandFired
  ;[1 1 1] On hand fired (left - right)
  (RangedWeaponDual|BindEventtoOnHandFired (delegate Custom|HandFired)))

(event EventBeginPlay
  (CallFunction|BindonHandFired))

(event Custom|HandFired
  (params (Hand :type EDualWeaponHand%20Enum :default "LEFT") (bPersistantFire :type Boolean))
  ;[1 1 1] On hand fired (left - right)
  (Variables|Default|SetLastHandFired Hand :!label _label)
  ;;[1 1 1] Comment
  (FortWeapon|PlayForceFeedback false)
  ;;[1 0 0.04289] TEMP!!! this is to solve  handfired being called even if shoot ability is blocked
  (Utilities|FlowControl|Gate true
    (:Exit
      (goto _label))))

(event EventOnPlayWeaponFireFX
  (params (bPersistentFire :type Boolean) (bSecondaryFire :type Boolean))
  (Utilities|FlowControl|Sequence
    (:then_0
      ;[1 1 1] On hand fired (left - right)
      ;;[1 0 0.04289] TEMP!!! this is to solve  handfired being called even if shoot ability is blocked
      (Utilities|FlowControl|Gate true :!label _label
        (:Exit
          (Variables|Default|SetLastHandFired (stale Custom|HandFired Hand))
          ;;[1 1 1] Comment
          (FortWeapon|PlayForceFeedback false)
          (goto _label Close))))
    (:then_1
      (FortWeapon|PlayForceFeedback false bPersistentFire))))

(event EventOnStopWeaponFireFX
  (FortWeapon|StopForceFeedback))

(event Custom|NewTarget
  (bind _lockonwidget (Variables|Default|GetLockOnWidget))
  (Utilities|IsValid _lockonwidget
    (:Is%20Valid
      (bind _returnvalue (Appearance|GetDynamicMaterial (Variables|/Game/Athena/Items/Gameplay/BackPacks/Ashton/Indigo/Ashton_Indigo_LockOnWidget.Ashton_Indigo_LockOnWidget_C|GetImage_83 _lockonwidget) :!class /Script/UMG.Image))
      (Utilities|FlowControl|Sequence
        (:then_0
          (bind (_direction _newtrack_0) (Timeline|CenterSize PlayFromStart)
            (:Update
              (Rendering|Material|SetScalarParameterValue _returnvalue "CenterSize" _newtrack_0 :!class /Script/Engine.MaterialInstanceDynamic))))
        (:then_1
          (bind (_direction_2 _newtrack_0_2) (Timeline|Intensity PlayFromStart)
            (:Update
              (Rendering|Material|SetScalarParameterValue _returnvalue "Intensity" _newtrack_0_2 :!class /Script/Engine.MaterialInstanceDynamic))))))))

(event FortWeapon|EventOnEquippedAndAttached
  ;[1 1 1] added bool
  (Variables|Default|SetIsIndigo true)
  (if (Pawn|IsLocallyControlled (Variables|Actor|GetInstigator))
    (bind _returnvalue (UserInterface|CreateWidget "/Game/Athena/Items/Gameplay/BackPacks/Ashton/Indigo/Ashton_Indigo_LockOnWidget.Ashton_Indigo_LockOnWidget_C"))
    (Variables|Default|SetLockOnWidget _returnvalue)
    (UserInterface|Viewport|AddtoViewport _returnvalue 0)
    (Utilities|FlowControl|Gate true
      (:Exit
        (Utilities|IsValid _returnvalue
          (:Is%20Valid
            (Variables|/Game/Athena/Items/Gameplay/BackPacks/Ashton/Indigo/Ashton_Indigo_LockOnWidget.Ashton_Indigo_LockOnWidget_C|SetTarget (Variables|Default|GetLockOnTarget) (Variables|Default|GetLockOnWidget))))))))

(event Custom|LockOnStart)

(event FortWeapon|EventOnLocalUnEquip
  ;[1 1 1] added bool
  (Variables|Default|SetIsIndigo false)
  (Widget|RemovefromParent (stale FortWeapon|EventOnEquippedAndAttached _returnvalue))
  (Utilities|FlowControl|Gate true
    (:Exit
      (Utilities|IsValid (stale FortWeapon|EventOnEquippedAndAttached _returnvalue)
        (:Is%20Valid
          (Variables|/Game/Athena/Items/Gameplay/BackPacks/Ashton/Indigo/Ashton_Indigo_LockOnWidget.Ashton_Indigo_LockOnWidget_C|SetTarget (Variables|Default|GetLockOnTarget) (Variables|Default|GetLockOnWidget)))))))

(event Custom|LockOnEnd
  (bind _lockonwidget (Variables|Default|GetLockOnWidget))
  (Utilities|IsValid _lockonwidget
    (:Is%20Valid
      (bind _returnvalue (Appearance|GetDynamicMaterial (Variables|/Game/Athena/Items/Gameplay/BackPacks/Ashton/Indigo/Ashton_Indigo_LockOnWidget.Ashton_Indigo_LockOnWidget_C|GetImage_83 _lockonwidget) :!class /Script/UMG.Image))
      (bind (_direction _newtrack_0) (Timeline|CenterReset ReverseFromEnd)
        (:Update
          (Rendering|Material|SetScalarParameterValue _returnvalue "CenterSize" _newtrack_0 :!class /Script/Engine.MaterialInstanceDynamic)
          (Rendering|Material|SetScalarParameterValue _returnvalue "Intensity" _newtrack_0 :!class /Script/Engine.MaterialInstanceDynamic))))))

(event EventTick
  (params (DeltaSeconds :type Float%20%28single-precision%29))
  (Utilities|FlowControl|Gate true
    (:Exit
      (Utilities|IsValid (stale FortWeapon|EventOnEquippedAndAttached _returnvalue)
        (:Is%20Valid
          (Variables|/Game/Athena/Items/Gameplay/BackPacks/Ashton/Indigo/Ashton_Indigo_LockOnWidget.Ashton_Indigo_LockOnWidget_C|SetTarget (Variables|Default|GetLockOnTarget) (Variables|Default|GetLockOnWidget)))))))

(event EventOnSetTargeting
  (params (bNewIsTargeting :type Boolean)))

(event FortWeapon|EventOnSpawnProjectile
  (params (SpawnProjectile :type Fort%20Projectile%20Base%20Object%20Reference))
  (bind _turnspeedmax (Variables|Default|GetTurnSpeedMax))
  (Networking|SwitchHasAuthority
    (:Authority
      (Utilities|IsValid (Variables|Default|GetLockOnTarget)
        (:Is%20Valid
          (bind _asb_prj_ashton_indigo (Utilities|Casting|CastToB_Prj_Ashton_Indigo SpawnProjectile))
          (Projectile|SetHomingTurnSpeedMin _asb_prj_ashton_indigo _asb_prj_ashton_indigo (- _turnspeedmax 1.0) :!class /Script/FortniteGame.FortProjectileMovementInterface)
          (Projectile|SetHomingTurnSpeedMax _asb_prj_ashton_indigo _asb_prj_ashton_indigo _turnspeedmax :!class /Script/FortniteGame.FortProjectileMovementInterface)
          (Projectile|SetHomingLockTargetDistanceThreshold _asb_prj_ashton_indigo _asb_prj_ashton_indigo (Variables|Default|GetHomingDistanceThreshold) :!class /Script/FortniteGame.FortProjectileMovementInterface)
          (Projectile|SetHomingTimeUntilMaxTurnSpeed _asb_prj_ashton_indigo _asb_prj_ashton_indigo (Variables|Default|GetTimeUntilMax) :!class /Script/FortniteGame.FortProjectileMovementInterface)
          (Projectile|SetHomingStyle _asb_prj_ashton_indigo _asb_prj_ashton_indigo "LockOn" true :!class /Script/FortniteGame.FortProjectileMovementInterface)
          (Projectile|SetHomingTarget _asb_prj_ashton_indigo _asb_prj_ashton_indigo (Variables|Default|GetLockOnTarget) 0 :!class /Script/FortniteGame.FortProjectileMovementInterface))))))

(fn HideReticle
  (bind _asfort_player_controller_zone (Utilities|Casting|CastToFortPlayerControllerZone (Game|GetInstigatorController)))
  (HUD|SetHUDElementVisibilitywithContext _asfort_player_controller_zone (Variables|Scope|GetReticleHUDElementTags) true "None" ""))

(fn ShowReticle
  (bind _asfort_player_controller_zone (Utilities|Casting|CastToFortPlayerControllerZone (Game|GetInstigatorController)))
  (HUD|ResetHUDElementVisibilitytoSettings _asfort_player_controller_zone (Variables|Scope|GetReticleHUDElementTags)))

(fn ConstructionScript)
