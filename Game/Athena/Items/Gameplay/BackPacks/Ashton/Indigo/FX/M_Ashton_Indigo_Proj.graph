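# Translucent, unlit, two-sided flame material for mesh particles, apparently for the Ashton Indigo backpack projectile effect.
# Fresnel blends two colors into HDR emissive, while three panning noise samples form a flame mask shaped by UV gradients and contrast.
# Vertex red switches inner and outer shell settings for shape, opacity and world position offset (inner along local -X, outer along normals).
# Opacity also uses inner-shell Fresnel, depth fade and particle alpha.
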Material /Game/Athena/Items/Gameplay/BackPacks/Ashton/Indigo/FX/M_Ashton_Indigo_Proj
graph {
 properties {
  bUsedWithMeshParticles=true
  bUsedWithNiagaraMeshParticles=true
  blendMode=BLEND_Translucent
  pixelDepthOffsetMode=PDOM_Legacy
  shadingModel=MSM_Unlit
  twoSided=true
  usedShadingModels=Unlit
 }
 nodes [
  n1 = Fresnel(ExponentIn=ScalarParameter() {defaultValue=2, parameterName=HDRContrastAdj})
  n2 = LinearInterpolate(
   A=VectorParameter() {group=FlameShape, parameterName=Color_B, defaultValue=<rgba 1 0.9083 0.5 1>}.RGB
   B=VectorParameter() {group=FlameShape, parameterName=Color_A, defaultValue=<rgba 5 0.5833 0 1>}.RGB
   Alpha=@n1
  )
  n3 = VectorParameter() {group=FlameShape, parameterName=HDR_Tint, defaultValue=<rgba 1 1 1 0>}
  MaterialFunctionCall(Emissive=MaterialFunctionCall(
    LDR Input=Clamp(None=@n2)
    HDR Tint=Multiply(A=@n2, B=@n3.RGB)
    HDR Intensity Multiplier=ScalarParameter() {defaultValue=5, group=FlameShape, parameterName=HDR_Intensity}
   ) {materialFunction=&DeriveHDRfromLDR}.RemappedRange) {materialFunction=&MF_FX_EmissiveAdjustments}
  n4 = TextureCoordinate()
  n5 = Power(Base=OneMinus(None=ComponentMask(None=@n4) {g=true}), Exp=ScalarParameter() {defaultValue=10, parameterName=Inner_WPOContrast})
  n6 = ComponentMask(None=VectorParameter() {group=FlameShape, parameterName=Flames_UVScale, defaultValue=<rgba 0.5 0.4 0 0>}.RGB) {g=true, r=true}
  n7 = Multiply(A=TextureCoordinate() {vTiling=0.5}, B=@n6)
  n8 = ScalarParameter() {group=FlameShape, parameterName=Flame_UVOffset}
  n9 = TextureSample(UVs=Panner(Coordinate=Add(A=@n7, B=@n8)) {speedX=-0.5, speedY=1}) {texture=&T_NoisePack_01}
  n10 = Constant() {r=0.2}
  n11 = TextureCoordinate() {uTiling=4}
  n12 = Multiply(A=Time(), B=ScalarParameter() {defaultValue=2, group=FlameShape, parameterName=Flame_Speed})
  n13 = Panner(Coordinate=Add(A=Multiply(A=@n6, B=@n11), B=@n8), Time=@n12) {speedX=0.2, speedY=1}
  n14 = Add(A=@n8, B=Multiply(A=@n6, B=TextureCoordinate() {uTiling=6, vTiling=2}))
  n15 = Multiply(
   A=Add(
    A=MaterialFunctionCall(In=@n9.G, Contrast=@n10) {materialFunction=&CheapContrast}.Result
    B=TextureSample(UVs=@n13) {texture=&T_NoisePack_01}.G
   )
   B=TextureSample(UVs=Panner(Coordinate=@n14, Time=@n12) {speedX=-0.2, speedY=1}) {texture=&T_NoisePack_01}.G
  )
  n16 = LinearInterpolate(Alpha=@n15)
  n17 = VertexColor()
  n18 = Transform(None=Constant3Vector() {constant=<rgba -1 0 0 0>}) {transformSourceType=TRANSFORMSOURCE_Local}
  n19 = VertexNormalWS()
  Multiply(
   A=Multiply(A=LinearInterpolate(A=@n5, B=@n16, Alpha=@n17.R), B=LinearInterpolate(A=@n18, B=@n19, Alpha=@n17.R))
   B=LinearInterpolate(
    A=ScalarParameter() {defaultValue=100, parameterName=Inner_WPOIntensity}
    B=ScalarParameter() {defaultValue=10, parameterName=Outer_WPOIntensity}
    Alpha=@n17.R
   )
  )
  n20 = TextureCoordinate()
  n21 = ScalarParameter() {defaultValue=2, parameterName=Inner_Flame_Contrast}
  n22 = ScalarParameter() {defaultValue=4, parameterName=Outer_Flame_Contrast}
  n23 = VertexColor()
  n24 = ScalarParameter() {defaultValue=5, group=FlameShape, parameterName=Inner_FireLine}
  n25 = ScalarParameter() {defaultValue=100000, group=FlameShape, parameterName=Outer_FireLine}
  n26 = VertexColor()
  n27 = Add(
   A=@n15
   B=Power(
    Base=MaterialFunctionCall() {materialFunction=&LinearGradient}.VGradient
    Exp=LinearInterpolate(A=@n24, B=@n25, Alpha=@n26.R)
   )
  )
  n28 = ScalarParameter() {group=FlameShape, parameterName=Inner_HardEdgeness}
  n29 = ScalarParameter() {defaultValue=0.5, group=FlameShape, parameterName=Outer_HardEdgeness}
  n30 = VertexColor()
  n31 = Clamp(None=MaterialFunctionCall(In=Power(Base=@n27) {constExponent=0.5}, Contrast=LinearInterpolate(A=@n28, B=@n29, Alpha=@n30.R)) {materialFunction=&CheapContrast}.Result)
  n32 = Multiply(A=@n31, B=Power(Base=ComponentMask(None=@n20) {g=true}, Exp=LinearInterpolate(A=@n21, B=@n22, Alpha=@n23.R)))
  n33 = ScalarParameter() {defaultValue=0.5, parameterName=Inner_Fresnel_Exp}
  n34 = LinearInterpolate(A=OneMinus(None=Fresnel(ExponentIn=@n33)), Alpha=@n23.R)
  Multiply(
   A=ParticleColor().A
   B=DepthFade(Opacity=Multiply(A=@n32, B=@n34), FadeDistance=ScalarParameter() {defaultValue=5, parameterName=Fade Distance})
  )
  Multiply(A=@n1, B=@n31)
 ]
}
