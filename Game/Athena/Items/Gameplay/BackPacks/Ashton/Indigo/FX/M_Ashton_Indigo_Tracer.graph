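# Translucent unlit bullet tracer material for the Ashton Indigo backpack effects.
# An eased timer from StartTime over WPODuration scales a red-vertex-masked offset from actor position toward BulletOrigin, stretching the tracer.
# That timer also drives fade-in and softens a UV gradient, which combines with local-X axis-aligned fresnel to shape opacity and dark-to-bright color.
# Particle color tints it; shading-path switches skip depth fade and HDR boosting on mobile.
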Material /Game/Athena/Items/Gameplay/BackPacks/Ashton/Indigo/FX/M_Ashton_Indigo_Tracer
graph {
 properties {
  blendMode=BLEND_Translucent
  pixelDepthOffsetMode=PDOM_Legacy
  previewMesh=/Game/Effects/Fort_Effects/Mesh/Shapes/SM_Projectile_Bullet
  shadingModel=MSM_Unlit
  usedShadingModels=Unlit
 }
 nodes [
  n1 = VectorParameter() {parameterName=BulletOrigin}
  n2 = ActorPositionWS()
  n3 = Divide(
   A=Subtract(A=Time() {period=5}, B=ScalarParameter() {parameterName=StartTime})
   B=ScalarParameter() {defaultValue=0.2, group=Timestamp, parameterName=WPODuration}
  )
  n4 = Clamp(None=MaterialFunctionCall!(In=Clamp(None=@n3)) {materialFunction=/Game/Effects/Fort_Effects/Materials/_Functions/MF_Smoothstep_EaseOut}.Result)
  Multiply(A=@n4, B=Multiply(A=VertexColor().R, B=Subtract(A=@n1.RGB, B=@n2)))
  n5 = ParticleColor()
  n6 = TextureCoordinate()
  n7 = Transform(None=Constant3Vector() {constant=<rgba 1 0 0 0>}) {transformSourceType=TRANSFORMSOURCE_Local}
  n8 = Multiply(
   A=Power(Base=ComponentMask(None=@n6) {g=true}, Exp=LinearInterpolate(Alpha=@n4) {constA=6, constAlpha=1})
   B=MaterialFunctionCall(Axis=Normalize(VectorInput=@n7), Exponent=ScalarParameter() {defaultValue=6, parameterName=FresnelExponent}) {materialFunction=&AxisAlignedFresnel}.Fresnel
  )
  n9 = Multiply(A=@n5.A, B=@n8)
  n10 = DepthFade(Opacity=@n9, FadeDistance=ScalarParameter() {defaultValue=40, parameterName=DepthFadeDist})
  n11 = Power(Base=@n4) {constExponent=0.3}
  Multiply(
   A=ScalarParameter() {defaultValue=1, parameterName=Opacity}
   B=Multiply(A=ShadingPathSwitch(Default=@n10, Mobile=@n9) {desc=Migrated from FeatureLevelSwitch}, B=Clamp(None=@n11))
  )
  n12 = Multiply(
   A=@n5.RGB
   B=LinearInterpolate(
    A=VectorParameter() {parameterName=DarkColor, defaultValue=<rgba 1 0.3866 0 0>}.RGB
    B=VectorParameter() {parameterName=BrightColor, defaultValue=<rgba 1.25 1.0906 0.625 0>}.RGB
    Alpha=@n8
   )
  )
  ShadingPathSwitch(
   Default=MaterialFunctionCall(LDR Input=Clamp(None=@n12), HDR Tint=@n12, HDR Intensity Multiplier=ScalarParameter() {defaultValue=30, parameterName=HDRMult}) {materialFunction=&DeriveHDRfromLDR}.RemappedRange
   Mobile=@n12
  ) {desc=Migrated from FeatureLevelSwitch}
 ]
}
