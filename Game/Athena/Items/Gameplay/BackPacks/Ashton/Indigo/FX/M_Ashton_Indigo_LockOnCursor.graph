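# Translucent UI material that draws the lock-on cursor reticle for the Ashton Indigo backpack item.
# The blue channel of the reticle mask, scaled about its center by an offset CenterSize parameter, is multiplied by CenterOpacity, a bright cyan-blue Reticle Color and Intensity.
# A vertically panning gradient, boosted and clamped to a nonzero floor, modulates the result to add a moving brightness sweep without blacking out the reticle.
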Material /Game/Athena/Items/Gameplay/BackPacks/Ashton/Indigo/FX/M_Ashton_Indigo_LockOnCursor
graph {
 properties {blendMode=BLEND_Translucent, materialDomain=MD_UI, pixelDepthOffsetMode=PDOM_Legacy, usedShadingModels=Default Lit}
 nodes [
  n1 = ScalarParameter() {defaultValue=0.55, parameterName=CenterSize}
  n2 = TextureSampleParameter2D(UVs=MaterialFunctionCall(UVs=TextureCoordinate(), Texture Scale=Subtract(A=@n1) {constB=0.11}) {materialFunction=&ScaleUVsByCenter}.UVs) {
   parameterName=Reticle
   samplerSource=SSM_Clamp_WorldGroupSettings
   texture=&/T_Ashton_Indigo_Mask
  }
  n3 = Multiply(A=@n2.B, B=ScalarParameter() {defaultValue=1, parameterName=CenterOpacity})
  n4 = VectorParameter() {parameterName=Reticle Color, defaultValue=<rgba 0.386 1.7893 3 1>}
  n5 = TextureSampleParameter2D(UVs=Panner(Coordinate=TextureCoordinate()) {speedY=0.25}) {parameterName=Panner, samplerType=SAMPLERTYPE_LinearColor, texture=/Game/Global/Textures/Default/Masks_and_Gradients/T_VertGradient}
  Multiply(
   A=Multiply(A=Multiply(A=@n3, B=@n4.RGB), B=ScalarParameter() {defaultValue=0.55, parameterName=Intensity})
   B=Clamp(None=Multiply(A=@n5.RGB) {constB=1.25}) {maxDefault=2, minDefault=0.8}
  )
 ]
}
